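# Niagara skydiving contrail cosmetic that trails multicolored ribbons, glows and lights from the player's skeleton.
# Enabled emitters layer bone-attached ribbons, faint sprites, stretching meshes and color-cycling lights, with front-end overrides for some velocities.
# Rising streak sprites with a light renderer scale spawn rate, length and upward speed from owner velocity through curves.
# An extra ribbon emitter and three other rising-sprite emitters are disabled.
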NiagaraSystem /BRCosmetics/Athena/Effects/Particles/Skydiving/NS_Contrail_HighTowerTapas/NS_Contrail_Informer_TowerTapas
dependencies {
 usedDataInterfaces [&NiagaraDataInterfaceCurve, &NiagaraDataInterfaceVectorCurve, &NiagaraDataInterfaceColorCurve]
 usedDynamicInputs []
 usedModules [&SystemState, /Niagara/Modules/System/SystemLifeCycle]
}
emitters {
 FN_AttachedParticle_v3 {
  emitterUpdateScript [
   &EmitterState()
   &SpawnBurst_Instantaneous(Spawn Count=200, Use Spawn Probability=true) {enabled=false}
   &SpawnRate(SpawnRate=4, Use Spawn Probability=false)
  ]
  particleSpawnScript [
   &FN_SetSpawnGroup()
   /Niagara/Modules/Spawn/Initialization/InitializeParticle(
    Module.Write Color=true
    Module.Write Lifetime=true
    Module.Write Mass=true
    Module.Write Material Random=true
    Module.Write Position=false
    Module.Write Scale=false
    Module.Write SpriteRotation=false
    Module.Write SpriteSize=true
    Position=&SimulationPosition()
    Sprite Size=<40 65>
   )
   /Niagara/Modules/Spawn/Location/SphereLocation(MaskSpawnGroup=false, Module.Override Randomness=false, Sphere Radius=500, Surface Only=false)
   &Color(
    Color=&UniformRangedLinearColor(Minimum=<rgba 1 0.2202 0 1>, Maximum=<rgba 0 1 0.7964 1>)
    Scale Alpha=0.03
   )
  ]
  particleUpdateScript [
   &ParticleState()
   &SkeletalMeshLocation(
    Bone / Socket Index=$Particles.SpawnGroup
    Bone Sampling Mode=::NewEnumerator4
    Inherit Velocity=false
    Interpolate Mesh Sampling=false
    MaskSpawnGroup=false
    Skeletal Mesh=DataInterface() {propertyValues {
      DefaultMesh=null
      ExcludeBoneName=root
      OptionalComponentClass=null
      PreviewMesh=&M_Med_Soldier_04_Preview
      SkinningMode=SkinOnTheFly
      SoftSourceActor=null
      SourceMode=Default
      UvSetIndex=0
      WholeMeshLOD=-1
      bExcludeBone=false
      bReadDeformedGeometry=true
      bRequireCurrentFrameData=true
      ComponentTags []
      FilteredBones [
       spine_01
       upperarm_twist_02_r
       upperarm_twist_02_l
       hand_r
       hand_l
       pelvis
       calf_r
       calf_l
       foot_r
       foot_l
       spine_03
       spine_05
       head
       neck_02
       thigh_twist_01_r
       thigh_twist_01_l
       thigh_l
       thigh_r
      ]
      FilteredSockets []
      MeshUserParameter {parameter {name=None, typeDefHandle {registeredTypeIndex=1}, varData []}}
      SamplingRegions []
     }}
   )
   &Color(
    Color=&UniformRangedLinearColor(Minimum=<rgba 1 0.3732 0.027 0.1>, Maximum=<rgba 0 0.8617 1 0.1>)
    Scale Alpha=&FloatFromCurve(FloatCurve=DataInterface() {propertyValues {Curve {keys [
         {arriveTangent=2.8779, interpMode=RCIM_Cubic, leaveTangent=2.8779, tangentMode=RCTM_User, time=0, value=0}
         <at 0.4 0.5 RCIM_Cubic>
         <at 1 0 RCIM_Cubic>
        ]}}})
   ) {module=Color001}
  ]
  rendererClasses [&NiagaraSpriteRendererProperties]
  renderers [{rendererClass=&NiagaraSpriteRendererProperties}]
 }
 FN_Contrails_Light {
  emitterSpawnScript [&FN_PreserveSkydiveUserVariables() {module=PreserveUserVariables}]
  emitterUpdateScript [
   &EmitterState()
   /Game/Effects/Niagara/Modules/Spawn/SpawnRateFromMultipleSpawnGroups_2(
    Global Spawnrate=4
    SpawnGroup0:NiagaraInt32=/Niagara/DynamicInputs/Random/RandomInt(Randomness Mode:ENiagaraRandomnessMode=::NewEnumerator0, Random Seed:NiagaraInt32=0, Fixed Random Seed:NiagaraBool=false)
   ) {enabled=false}
   &SpawnBurst_Instantaneous(Spawn Count=4, Use Spawn Probability=false) {enabled=false}
   &SpawnRate(SpawnRate=5, Use Spawn Probability=false)
  ]
  particleSpawnScript [
   /SetVariables(Particles.LightExponent=1) {isSetParameters=true}
   /Game/Effects/Niagara/Modules/Groups/FN_SetSpawnGroup_v1/FN_SetSpawnGroup(Emitter Bone Count=8)
   &InitializeParticle(
    Module.Write Color=true
    Module.Write Lifetime=true
    Module.Write Mass=true
    Module.Write Material Random=true
    Module.Write Position=false
    Module.Write Scale=false
    Module.Write SpriteRotation=false
    Module.Write SpriteSize=true
    Position=&SimulationPosition()
    Sprite Size=&Vector2DFromFloat(Value=&UniformRangedFloat(Minimum=5, Maximum=10))
   )
  ]
  particleUpdateScript [
   /SetVariables(Particles.Color=&MakeLinearColorFromVectorAndFloat(
     "Vector (RGB)"=&MakeVectorFromLinearColorRGB(Color=&ColorFromCurve(ColorCurve=DataInterface() {propertyValues {
         AlphaCurve {keys [<at -0.0128 0>,<at 0.3602 1 RCIM_Cubic>,<at 0.9977 0 RCIM_Cubic>]}
         BlueCurve {keys [<at -0.0036 0.0791 RCIM_Cubic>,<at 0.341 0.1071>,<at 0.5867 1>,<at 0.8035 1 RCIM_Cubic>]}
         GreenCurve {keys [<at -0.0036 0.1156 RCIM_Cubic>,<at 0.341 1>,<at 0.5867 0.7902>,<at 0.8035 0 RCIM_Cubic>]}
         RedCurve {keys [<at -0.0036 1 RCIM_Cubic>,<at 0.341 0.7069>,<at 0.5867 0>,<at 0.8035 0.1441 RCIM_Cubic>]}
        }}))
     "Float (Alpha)"=&FloatFromCurve(FloatCurve=DataInterface() {propertyValues {Curve {keys [<at 0 0>,<at 0.4341 0.5241>,<at 1.0011 0.002>]}}})
    )) {isSetParameters=true}
   &ParticleState()
   &SampleSkeletalMeshSkeleton(
    Mesh=DataInterface() {propertyValues {
      DefaultMesh=null
      ExcludeBoneName=root
      OptionalComponentClass=null
      PreviewMesh=&M_Med_Soldier_04_Preview
      SkinningMode=SkinOnTheFly
      SoftSourceActor=null
      SourceMode=Default
      UvSetIndex=0
      WholeMeshLOD=-1
      bExcludeBone=false
      bReadDeformedGeometry=true
      bRequireCurrentFrameData=true
      ComponentTags []
      FilteredBones [hand_l,hand_r,spine_03,root,foot_l,foot_r]
      FilteredSockets []
      MeshUserParameter {parameter {name=None, typeDefHandle {registeredTypeIndex=1}, varData []}}
      SamplingRegions []
     }}
    Specific Bone Index:NiagaraInt32=$Particles.SpawnGroup
   )
   /Niagara/Modules/Spawn/Location/SkeletalMeshSkeletonLocation(Module.Inherit Velocity=false, Module.Use Sampled Bone as Ribbon ID=true, Sampled Position Offset:Vector3f=<0 0 10>)
   &SolveForcesAndVelocity(Clamp Velocity=false, Limit Acceleration=false) {enabled=false}
   &FN_Light_v3(
    Color=&MakeLinearColorFromVectorAndFloat(
     "Vector (RGB)"=&MakeVectorFromLinearColorRGB(Color=&ColorFromCurve(ColorCurve=DataInterface() {propertyValues {
         AlphaCurve {keys [<at 0 1>,<at 1 1>]}
         BlueCurve {keys [<at 0 0>,<at 0.3143 0>,<at 0.603 0.3061>,<at 0.8468 1>,<at 1 0>]}
         GreenCurve {keys [<at 0 0>,<at 0.3143 0.5246>,<at 0.603 1>,<at 0.8468 0>,<at 1 0>]}
         RedCurve {keys [<at 0 0>,<at 0.3143 1>,<at 0.603 0>,<at 0.8468 0.1331>,<at 1 0>]}
        }}))
     "Float (Alpha)"=&FloatFromCurve(FloatCurve=DataInterface() {propertyValues {Curve {keys [<at 0 0>,<at 0.2017 1>,<at 1 0>]}}})
    )
    Create Color Attribute=true
    Scale Color=10
   )
  ]
  rendererClasses [&NiagaraLightRendererProperties]
  renderers [{rendererClass=&NiagaraLightRendererProperties, rendererIndex=0}]
 }
 FN_Contrails_RibbonWde {
  emitterUpdateScript [
   &EmitterScalability(
    Module.bApplySpawnCountScale=false
    Module.bApplySpawnCountScaleByDistanceFraction=false
    Module.bUseMaxDistance=false
    Module.bUseMinDistance=false
   ) {enabled=false}
   &EmitterLifeCycle(CompleteOnInactive=true) {enabled=false}
   &EmitterState()
   /Game/Effects/Niagara/Modules/Spawn/FN_SpawnContrails(All limbs=false, Global Spawnrate=10) {enabled=false}
   &FN_SpawnRibbon_v2(Number of Ribbons=4, Spawn Mode=::NewEnumerator0, SpawnRate=2)
  ]
  particleSpawnScript [
   &FN_InitializeRibbon_v3(Enable Skeletal Position=true, Lifetime=1, Ribbon Width=40)
   &AddVelocity(Velocity=<0 0 410>) {enabled=false}
   &AddVelocity(Velocity=&FN_FrontEndVector3Modifier(Gameplay=<0 0 410>, Front End=<0 0 350>)) {module=AddVelocity001}
  ]
  particleUpdateScript [
   &AccelerationForce(
    Acceleration=&VectorFromCurve(VectorCurve=DataInterface() {propertyValues {
       XCurve {keys [<at 0.0017 132.1921>,<at 0.9905 -65.3336>]}
       YCurve {keys [<at -0.0011 1.0067>,<at 0.9988 0>]}
       ZCurve {keys [<at -0.0011 -300>,<at 0.9957 -200.05>]}
      }})
    Coordinate Space=::Local
   ) {enabled=false}
   &Drag(Drag=0.9, Use Linear Drag=true, Use Rotational Drag=false)
   &InheritVelocity(
    Module.Limit Speed=true
    Velocity Limit=10000
    Velocity Scale=<-0.03 -0.03 -0.01>
    Velocity Source=&TransformVector(Vector=$Engine.Owner.Velocity, Source Space=::World, Destination Space=::Local)
   ) {enabled=false}
   /Game/Effects/Niagara/Modules/Skydiving/FN_DivingAttributes/FN_DivingAttributes()
   &SolveForcesAndVelocity(Clamp Velocity=false, Limit Acceleration=false)
   &FN_UpdateRibbon(
    Enable Skeletal Position=true
    Initial Color:LinearColor=$Particles.Initial.Color
    Initial RibbonWidth:NiagaraFloat=$Particles.Initial.RibbonWidth
   )
   &Material(Base Texture Time:NiagaraFloat=&Multiply_Float(A=$Emitter.Age, B=0)) {enabled=false}
   &DynamicMaterialParameters(Module.Param0WriteEnabled=true, Module.Param1WriteEnabled=true, Module.Param2WriteEnabled=true, Module.Param3WriteEnabled=true) {enabled=false}
   &ScaleRibbonWidth(Ribbon Width Scale=&FloatFromCurve(FloatCurve=DataInterface() {propertyValues {Curve {keys [
         {arriveTangent=1.8003, interpMode=RCIM_Cubic, leaveTangent=0.4952, tangentMode=RCTM_User, time=0.0089, value=0.7995}
         {leaveTangent=0.9504, time=0.4538, value=1}
        ]}}}))
   &Color(
    Color=&ColorFromCurve(ColorCurve=DataInterface() {propertyValues {
       AlphaCurve {keys [<at 0.0048 0.15>,<at 0.2601 0.05>,<at 0.9857 -0.003 RCIM_Cubic>]}
       BlueCurve {keys [<at -0.0023 0.0904 RCIM_Cubic>,<at 0.0555 0.0084>,<at 0.1428 1>,<at 0.3271 0.8273>,<at 0.6292 0.6799>]}
       GreenCurve {keys [<at -0.0023 0.4486 RCIM_Cubic>,<at 0.0555 0.0682>,<at 0.1428 0.4441>,<at 0.3271 0.151>,<at 0.6292 0>]}
       RedCurve {keys [<at -0.0023 1 RCIM_Cubic>,<at 0.0555 1>,<at 0.1428 0.0677>,<at 0.3271 1>,<at 0.6292 1>]}
      }})
    Scale Alpha=&FloatFromCurve(
     FloatCurve=DataInterface() {propertyValues {Curve {keys [{arriveTangent=0.27, interpMode=RCIM_Cubic, leaveTangent=0.27, tangentMode=RCTM_User, time=-0.0046, value=1.0113}, <at 0.6 0.002>]}}}
     Scale Curve=0.8
    )
   )
   &DynamicMaterialParameters(
    Index 0 Param 1=1
    Index 0 Param 2=1
    Index 0 Param 3=0.1
    Index 0 Param 4=0.2
    Index 1 Param 1=&Add_Float(
     A=&Multiply_Float(A=1, B=&MakeFloatFromInt(Int=HlslExpression() {hlslExpression=Particles.RibbonID}))
     B=&Multiply_Float(A=-1, B=$Engine.Time)
    )
    Index 1 Param 3=1
    Index 1 Param 4=1
    Index 2 Param 1=0.5
    Index 2 Param 2=0.5
    Index 2 Param 3=0.5
    Index 3 Param 1=1
    Index 3 Param 2=1
    Index 3 Param 3=1
    Index 3 Param 4=1
    Module.Param0WriteEnabled=true
    Module.Param0WriteEnabled1=true
    Module.Param0WriteEnabled2=true
    Module.Param0WriteEnabled3=true
    Module.Param1WriteEnabled=true
    Module.Param1WriteEnabled1=true
    Module.Param1WriteEnabled2=true
    Module.Param1WriteEnabled3=true
    Module.Param2WriteEnabled=true
    Module.Param2WriteEnabled1=true
    Module.Param2WriteEnabled2=true
    Module.Param2WriteEnabled3=true
    Module.Param3WriteEnabled=true
    Module.Param3WriteEnabled1=true
    Module.Param3WriteEnabled2=true
    Module.Param3WriteEnabled3=true
    Write Parameter Index 1=true
    Write Parameter Index 2=true
    Write Parameter Index 3=true
   ) {module=DynamicMaterialParameters001}
  ]
  rendererClasses [&NiagaraRibbonRendererProperties]
  renderers [{rendererClass=&NiagaraRibbonRendererProperties}]
 }
 FN_Contrails_RibbonWde001 {
  bEnabled=false
  emitterUpdateScript [
   &EmitterScalability(
    Module.bApplySpawnCountScale=false
    Module.bApplySpawnCountScaleByDistanceFraction=false
    Module.bUseMaxDistance=false
    Module.bUseMinDistance=false
   ) {enabled=false}
   &EmitterLifeCycle(CompleteOnInactive=true) {enabled=false}
   &EmitterState()
   &FN_SpawnContrails(All limbs=false, Global Spawnrate=10) {enabled=false}
   &FN_SpawnRibbon_v2(Number of Ribbons=20, Spawn Mode=::NewEnumerator0, SpawnRate=2)
  ]
  particleSpawnScript [
   &FN_InitializeRibbon_v3(Enable Skeletal Position=true, Lifetime=1, Offset Amount=11, Ribbon Width=30)
   &AddVelocity(Velocity=<0 0 410>) {enabled=false}
   &AddVelocity(Velocity=&FN_FrontEndVector3Modifier(Gameplay=<0 0 410>, Front End=<0 0 3000>)) {module=AddVelocity001}
  ]
  particleUpdateScript [
   &AccelerationForce(
    Acceleration=&VectorFromCurve(VectorCurve=DataInterface() {propertyValues {
       XCurve {keys [<at 0.0017 132.1921>,<at 0.9905 -65.3336>]}
       YCurve {keys [<at -0.0011 1.0067>,<at 0.9988 0>]}
       ZCurve {keys [<at -0.0011 -300>,<at 0.9957 -200.05>]}
      }})
    Coordinate Space=::Local
   ) {enabled=false}
   &Drag(Drag=0.9, Use Linear Drag=true, Use Rotational Drag=false)
   &InheritVelocity(
    Module.Limit Speed=true
    Velocity Limit=10000
    Velocity Scale=<-0.03 -0.03 -0.01>
    Velocity Source=&TransformVector(Vector=$Engine.Owner.Velocity, Source Space=::World, Destination Space=::Local)
   ) {enabled=false}
   &FN_DivingAttributes()
   &SolveForcesAndVelocity(Clamp Velocity=false, Limit Acceleration=false)
   &FN_UpdateRibbon(
    Enable Skeletal Position=true
    Initial Color:LinearColor=$Particles.Initial.Color
    Initial RibbonWidth:NiagaraFloat=$Particles.Initial.RibbonWidth
   )
   &Material(Base Texture Time:NiagaraFloat=&Multiply_Float(A=$Emitter.Age, B=0)) {enabled=false}
   &DynamicMaterialParameters(Module.Param0WriteEnabled=true, Module.Param1WriteEnabled=true, Module.Param2WriteEnabled=true, Module.Param3WriteEnabled=true) {enabled=false}
   &ScaleRibbonWidth(Ribbon Width Scale=&FloatFromCurve(FloatCurve=DataInterface() {propertyValues {Curve {keys [
         {arriveTangent=1.8003, interpMode=RCIM_Cubic, leaveTangent=0.4952, tangentMode=RCTM_User, time=0.0089, value=0.7995}
         {leaveTangent=0.9504, time=0.4538, value=1}
        ]}}}))
   &Color(
    Color=&ColorFromCurve(ColorCurve=DataInterface() {propertyValues {
       AlphaCurve {keys [<at 0.0048 0.6>,<at 0.2601 0.5>,<at 0.459 0.5>,<at 0.9857 -0.003 RCIM_Cubic>]}
       BlueCurve {keys [<at -0.0023 0.0904 RCIM_Cubic>,<at 0.0555 0.0084>,<at 0.1428 1>,<at 0.3271 0.8273>,<at 0.6292 0.6799>]}
       GreenCurve {keys [<at -0.0023 0.4486 RCIM_Cubic>,<at 0.0555 0.0682>,<at 0.1428 0.4441>,<at 0.3271 0.151>,<at 0.6292 0>]}
       RedCurve {keys [<at -0.0023 1 RCIM_Cubic>,<at 0.0555 1>,<at 0.1428 0.0677>,<at 0.3271 1>,<at 0.6292 1>]}
      }})
    Scale Alpha=&FloatFromCurve(
     FloatCurve=DataInterface() {propertyValues {Curve {keys [{arriveTangent=0.27, interpMode=RCIM_Cubic, leaveTangent=0.27, tangentMode=RCTM_User, time=-0.0046, value=1.0113}, <at 0.6 0.002>]}}}
     Scale Curve=0.8
    )
   )
   &DynamicMaterialParameters(
    Index 0 Param 1=1
    Index 0 Param 2=1
    Index 0 Param 3=0.1
    Index 0 Param 4=0.2
    Index 1 Param 1=&Add_Float(
     A=&Multiply_Float(A=1, B=&MakeFloatFromInt(Int=HlslExpression() {hlslExpression=Particles.RibbonID}))
     B=&Multiply_Float(A=-1, B=$Engine.Time)
    )
    Index 1 Param 3=1
    Index 1 Param 4=1
    Index 2 Param 1=0.5
    Index 2 Param 2=0.5
    Index 2 Param 3=0.5
    Index 3 Param 1=1
    Index 3 Param 2=1
    Index 3 Param 3=1
    Index 3 Param 4=1
    Module.Param0WriteEnabled=true
    Module.Param0WriteEnabled1=true
    Module.Param0WriteEnabled2=true
    Module.Param0WriteEnabled3=true
    Module.Param1WriteEnabled=true
    Module.Param1WriteEnabled1=true
    Module.Param1WriteEnabled2=true
    Module.Param1WriteEnabled3=true
    Module.Param2WriteEnabled=true
    Module.Param2WriteEnabled1=true
    Module.Param2WriteEnabled2=true
    Module.Param2WriteEnabled3=true
    Module.Param3WriteEnabled=true
    Module.Param3WriteEnabled1=true
    Module.Param3WriteEnabled2=true
    Module.Param3WriteEnabled3=true
    Write Parameter Index 1=true
    Write Parameter Index 2=true
    Write Parameter Index 3=true
   ) {module=DynamicMaterialParameters001}
  ]
  rendererClasses [&NiagaraRibbonRendererProperties]
  renderers [{rendererClass=&NiagaraRibbonRendererProperties}]
 }
 FN_NE_Skydive_Mesh {
  emitterSpawnScript [&FN_PreserveSkydiveUserVariables() {module=PreserveSkydiveUserVariables}]
  emitterUpdateScript [
   &EmitterScalability(
    Module.bApplySpawnCountScale=false
    Module.bApplySpawnCountScaleByDistanceFraction=false
    Module.bUseMaxDistance=false
    Module.bUseMinDistance=false
   )
   &EmitterState()
   &EmitterLifeCycle(AutoComplete=false, CompleteOnInactive=true)
   /Game/Effects/Niagara/Modules/Spawn/FN_SkydiveSpawnRate(Global Spawnrate=7, SpawnGroup0=4, SpawnGroup1=4, SpawnGroup2=4, SpawnGroup3=4) {enabled=false, module=SkydiveSpawnRate}
   &SpawnRate(SpawnGroup=4, SpawnRate=2, Use Spawn Probability=false)
  ]
  particleSpawnScript [
   &FN_SetSpawnGroup(Override SpawnGroup=true, SpawnGroup:NiagaraInt32=&SpawnGroup())
   &InitializeParticle(
    Lifetime=&UniformRangedFloat(Minimum=0.5, Maximum=1.5)
    Mesh Scale=&UniformRangedVector(Minimum=<1 1 5>, Maximum=<3 3 20>)
    Module.Write Color=true
    Module.Write Lifetime=true
    Module.Write Mass=false
    Module.Write Material Random=true
    Module.Write Position=false
    Module.Write Scale=true
    Module.Write SpriteRotation=false
    Module.Write SpriteSize=false
    Position=&SimulationPosition()
   )
   &SampleSkeletalMeshSkeleton(
    Mesh:NiagaraDataInterfaceSkeletalMesh=$User.SkeletalMesh
    Specific Bone Index:NiagaraInt32=$Particles.SpawnGroup
    Specific Socket Index:NiagaraInt32=$Particles.SpawnGroup
   )
   &SkeletalMeshSkeletonLocation(Module.Inherit Velocity=false, Module.Use Sampled Bone as Ribbon ID=true, Sampled Rotation Attribute:Quat4f=$Engine.Owner.Rotation)
   &ApplyInitialForces() {enabled=false}
   /Niagara/Modules/Spawn/Orientation/V2/InitialMeshOrientation(
    Orientation Vector=<0 0 0>
    Rotation=&UniformRangedVector(Minimum=<0 0 0>, Maximum=<0 0 1>)
    Use Orientation Vector=true
    Use Rotation Vector=true
   )
   &BoxLocation(Box Offset=<0 0 -100>, Box Size=<0 0 0>, MaskSpawnGroup=false, Module.Override Seed=false, Rotate Box=false, Surface Only=false)
   &AddVelocity(Velocity=&FN_FrontEndVector3Modifier(Gameplay=<0 0 10>, Front End=<0 0 200>))
  ]
  particleUpdateScript [
   &ParticleState()
   &SolveForcesAndVelocity(Clamp Velocity=false, Limit Acceleration=false)
   &Color(
    Color=&MakeLinearColor(
     R:NiagaraFloat=1
     G:NiagaraFloat=1
     B:NiagaraFloat=1
     A:NiagaraFloat=&FloatFromCurve(
      FloatCurve=DataInterface() {propertyValues {Curve {keys [
          {arriveTangent=6.9452, interpMode=RCIM_Cubic, leaveTangent=6.9452, tangentMode=RCTM_User, time=0, value=0}
          {arriveTangent=0.8454, interpMode=RCIM_Cubic, leaveTangent=0.8454, time=0.2438, value=2}
          {arriveTangent=-2.6448, interpMode=RCIM_Cubic, leaveTangent=-2.6448, time=0.47, value=0.3973}
          <at 1 0 RCIM_Cubic>
         ]}}}
      Scale Curve=2
     )
    )
    Scale Alpha=2
   )
   &ScaleMeshSize(Scale Factor=&MakeVector(
     X=1
     Y=1
     Z=&Multiply_Float(A=1, B=&FloatFromCurve(FloatCurve=DataInterface() {propertyValues {Curve {keys [<at 0 1>,<at 1 7>]}}}))
    ))
   &DynamicMaterialParameters(
    Index 0 Param 2=2
    Module.Param0WriteEnabled=true
    Module.Param1WriteEnabled=true
    Module.Param2WriteEnabled=true
    Module.Param3WriteEnabled=true
   )
  ]
  rendererClasses [&NiagaraMeshRendererProperties]
  renderers [{rendererClass=&NiagaraMeshRendererProperties, rendererIndex=0}]
 }
 Little_Round_Drops_Up {
  bEnabled=false
  emitterSpawnScript [&FN_PreserveSkydiveUserVariables() {enabled=false, module=PreserveUserVariables}]
  emitterUpdateScript [
   &EmitterState(Inactive Response=::NewEnumerator1, Life Cycle Mode=::NewEnumerator1, Loop Behavior=::NewEnumerator1, Loop Duration=3600, UseLoopDelay=false)
   &SpawnRate(SpawnRate=50, Use Spawn Probability=false)
  ]
  particleSpawnScript [
   &FN_SetSpawnGroup(Emitter Bone Count=8)
   &InitializeParticle(
    Lifetime=&UniformRangedFloat(Minimum=0.4, Maximum=0.8)
    Module.Write Color=true
    Module.Write Lifetime=true
    Module.Write Mass=true
    Module.Write Material Random=true
    Module.Write Position=false
    Module.Write Scale=false
    Module.Write SpriteRotation=false
    Module.Write SpriteSize=false
    Position=&SimulationPosition()
   )
   /SetVariables(Particles.Scale=&VectorFromFloat(Value=100)) {enabled=false, isSetParameters=true}
   &SampleSkeletalMeshSkeleton(Mesh:NiagaraDataInterfaceSkeletalMesh=$User.SkeletalMesh, Specific Bone Index:NiagaraInt32=$Particles.SpawnGroup)
   &SkeletalMeshSkeletonLocation(Module.Inherit Velocity=false, Module.Use Sampled Bone as Ribbon ID=true)
   /SetVariables(Particles.Velocity=&UniformRangedVector(Minimum=<-25 -25 25>, Maximum=<25 25 50>)) {enabled=false, isSetParameters=true, module=SetVariables_#1}
   /SetVariables(Particles.SpriteSize=&UniformRangedVector2D(Minimum=<50 50>, Maximum=<100 200>)) {isSetParameters=true, module=SetVariables_#2}
   &SphereLocation(
    Coordinate Space:ENiagaraCoordinateSpace=::Simulation
    MaskSpawnGroup=false
    Module.Override Randomness=false
    Sphere Radius=10
    Surface Only=false
   )
  ]
  particleUpdateScript [
   &ParticleState()
   &GravityForce() {enabled=false}
   &SolveForcesAndVelocity(Clamp Velocity=false, Limit Acceleration=false)
   /SetVariables(Particles.Velocity=/Niagara/DynamicInputs/ScaleByCurve/ScaleVectorByCurve(
     InitialValue:Vector3f=&UniformRangedVector(Minimum=<0 0 0.5>, Maximum=<0 0 1>)
     ScaleCurve:NiagaraDataInterfaceVectorCurve=DataInterface() {propertyValues {
       XCurve {keys [<at 0 50 RCIM_Cubic>, {arriveTangent=-241.6722, interpMode=RCIM_Cubic, leaveTangent=-241.672, tangentMode=RCTM_User, time=0.5, value=20}]}
       YCurve {keys [<at 0 50 RCIM_Cubic>, {arriveTangent=-247.443, interpMode=RCIM_Cubic, leaveTangent=-247.448, tangentMode=RCTM_User, time=0.4991, value=20}]}
       ZCurve {keys [<at 0 100>,<at 1 500>]}
      }}
     CurveIndex:NiagaraFloat=$Particles.NormalizedAge
    )) {isSetParameters=true, module=SetVariables_#3}
   /SetVariables(Particles.Velocity=<0 0 -980>) {enabled=false, isSetParameters=true, module=SetVariables_#4}
   &Color(Color=&ColorFromCurve(ColorCurve=DataInterface() {propertyValues {
       AlphaCurve {keys [<at 0 0.1 RCIM_Cubic>,<at 1 0 RCIM_Cubic>]}
       BlueCurve {keys [<at 0 1 RCIM_Cubic>,<at 0.5474 1>,<at 1 1 RCIM_Cubic>]}
       GreenCurve {keys [<at 0 1 RCIM_Cubic>,<at 0.5474 1>,<at 1 1 RCIM_Cubic>]}
       RedCurve {keys [<at 0 1 RCIM_Cubic>,<at 0.5474 1>,<at 1 1 RCIM_Cubic>]}
      }}))
  ]
  rendererClasses [&NiagaraSpriteRendererProperties]
  renderers [{rendererClass=&NiagaraSpriteRendererProperties}]
 }
 Little_Round_Drops_Up001 {
  bEnabled=false
  emitterSpawnScript [&FN_PreserveSkydiveUserVariables() {enabled=false, module=PreserveUserVariables}]
  emitterUpdateScript [
   &EmitterState(Inactive Response=::NewEnumerator1, Life Cycle Mode=::NewEnumerator1, Loop Behavior=::NewEnumerator1, Loop Duration=3600, UseLoopDelay=false)
   &SpawnRate(SpawnRate=10, Use Spawn Probability=false)
  ]
  particleSpawnScript [
   &FN_SetSpawnGroup(Emitter Bone Count=8) {enabled=false}
   &InitializeParticle(
    Lifetime=&UniformRangedFloat(Minimum=0.5, Maximum=1.2)
    Module.Write Color=true
    Module.Write Lifetime=true
    Module.Write Mass=true
    Module.Write Material Random=true
    Module.Write Position=false
    Module.Write Scale=false
    Module.Write SpriteRotation=false
    Module.Write SpriteSize=false
    Position=&SimulationPosition()
   )
   &SampleSkeletalMeshSkeleton(
    Mesh=DataInterface() {propertyValues {
      DefaultMesh=null
      ExcludeBoneName=root
      OptionalComponentClass=null
      PreviewMesh=&M_Med_Soldier_04_Preview
      SkinningMode=SkinOnTheFly
      SoftSourceActor=null
      SourceMode=Default
      UvSetIndex=0
      WholeMeshLOD=-1
      bExcludeBone=false
      bReadDeformedGeometry=true
      bRequireCurrentFrameData=true
      ComponentTags []
      FilteredBones [spine_02]
      FilteredSockets []
      MeshUserParameter {parameter {name=None, typeDefHandle {registeredTypeIndex=1}, varData []}}
      SamplingRegions []
     }}
    Specific Bone Index:NiagaraInt32=$Particles.SpawnGroup
   )
   &SkeletalMeshSkeletonLocation(Module.Inherit Velocity=false, Module.Use Sampled Bone as Ribbon ID=true)
   /SetVariables(Particles.Velocity=&UniformRangedVector(Minimum=<-25 -25 25>, Maximum=<25 25 50>)) {enabled=false, isSetParameters=true, module=SetVariables_#1}
   /SetVariables(Particles.SpriteSize=&UniformRangedVector2D(Minimum=<80 400>, Maximum=<100 1000>)) {isSetParameters=true, module=SetVariables_#2}
   &SphereLocation(
    Coordinate Space:ENiagaraCoordinateSpace=::Simulation
    MaskSpawnGroup=false
    Module.Override Randomness=false
    Sphere Radius=10
    Surface Only=false
   )
   &BoxLocation(
    Box Offset=&UniformRangedVector(Minimum=<0 0 -30>, Maximum=<1 1 30>)
    Box Size=<80 80 80>
    MaskSpawnGroup=false
    Module.Override Seed=false
    Rotate Box=false
    Surface Only=false
   )
  ]
  particleUpdateScript [
   &ParticleState()
   &GravityForce() {enabled=false}
   &SolveForcesAndVelocity(Clamp Velocity=false, Limit Acceleration=false)
   /SetVariables(Particles.Velocity=&ScaleVectorByCurve(
     InitialValue:Vector3f=&UniformRangedVector(Minimum=<0 0 0.5>, Maximum=<0 0 1>)
     ScaleCurve:NiagaraDataInterfaceVectorCurve=DataInterface() {propertyValues {
       XCurve {keys [<at 0 50 RCIM_Cubic>, {arriveTangent=-241.6722, interpMode=RCIM_Cubic, leaveTangent=-241.672, tangentMode=RCTM_User, time=0.5, value=20}]}
       YCurve {keys [<at 0 50 RCIM_Cubic>, {arriveTangent=-247.443, interpMode=RCIM_Cubic, leaveTangent=-247.448, tangentMode=RCTM_User, time=0.4991, value=20}]}
       ZCurve {keys [<at 0 1600>,<at 1 800>]}
      }}
     CurveIndex:NiagaraFloat=$Particles.NormalizedAge
    )) {isSetParameters=true, module=SetVariables_#3}
   /SetVariables(Particles.Velocity=<0 0 -980>) {enabled=false, isSetParameters=true, module=SetVariables_#4}
   &Color(Color=&ColorFromCurve(ColorCurve=DataInterface() {propertyValues {
       AlphaCurve {keys [<at 0 0.3 RCIM_Cubic>,<at 1 0 RCIM_Cubic>]}
       BlueCurve {keys [<at 0 1 RCIM_Cubic>,<at 0.5474 1>,<at 1 1 RCIM_Cubic>]}
       GreenCurve {keys [<at 0 1 RCIM_Cubic>,<at 0.5474 1>,<at 1 1 RCIM_Cubic>]}
       RedCurve {keys [<at 0 1 RCIM_Cubic>,<at 0.5474 1>,<at 1 1 RCIM_Cubic>]}
      }}))
  ]
  rendererClasses [&NiagaraSpriteRendererProperties]
  renderers [{rendererClass=&NiagaraSpriteRendererProperties}]
 }
 Little_Round_Drops_Up002 {
  bEnabled=false
  emitterSpawnScript [&FN_PreserveSkydiveUserVariables() {enabled=false, module=PreserveUserVariables}]
  emitterUpdateScript [
   &EmitterState(Inactive Response=::NewEnumerator1, Life Cycle Mode=::NewEnumerator1, Loop Behavior=::NewEnumerator1, Loop Duration=3600, UseLoopDelay=false)
   &SpawnRate(Use Spawn Probability=false)
  ]
  particleSpawnScript [
   &FN_SetSpawnGroup(Emitter Bone Count=8) {enabled=false}
   &InitializeParticle(
    Lifetime=&UniformRangedFloat(Minimum=0.5, Maximum=1)
    Module.Write Color=true
    Module.Write Lifetime=true
    Module.Write Mass=true
    Module.Write Material Random=true
    Module.Write Position=false
    Module.Write Scale=false
    Module.Write SpriteRotation=false
    Module.Write SpriteSize=false
    Position=&SimulationPosition()
   )
   &SampleSkeletalMeshSkeleton(
    Mesh=DataInterface() {propertyValues {
      DefaultMesh=null
      ExcludeBoneName=root
      OptionalComponentClass=null
      PreviewMesh=&M_Med_Soldier_04_Preview
      SkinningMode=SkinOnTheFly
      SoftSourceActor=null
      SourceMode=Default
      UvSetIndex=0
      WholeMeshLOD=-1
      bExcludeBone=false
      bReadDeformedGeometry=true
      bRequireCurrentFrameData=true
      ComponentTags []
      FilteredBones [spine_02]
      FilteredSockets []
      MeshUserParameter {parameter {name=None, typeDefHandle {registeredTypeIndex=1}, varData []}}
      SamplingRegions []
     }}
    Specific Bone Index:NiagaraInt32=$Particles.SpawnGroup
   )
   &SkeletalMeshSkeletonLocation(Module.Inherit Velocity=false, Module.Use Sampled Bone as Ribbon ID=true)
   /SetVariables(Particles.Velocity=&UniformRangedVector(Minimum=<-25 -25 25>, Maximum=<25 25 50>)) {enabled=false, isSetParameters=true, module=SetVariables_#1}
   /SetVariables(Particles.SpriteSize=&UniformRangedVector2D(Minimum=<200 500>, Maximum=<400 1000>)) {isSetParameters=true, module=SetVariables_#2}
   &SphereLocation(
    Coordinate Space:ENiagaraCoordinateSpace=::Simulation
    MaskSpawnGroup=false
    Module.Override Randomness=false
    Sphere Radius=10
    Surface Only=false
   )
   &BoxLocation(
    Box Offset=&UniformRangedVector(Minimum=<0 0 -30>, Maximum=<1 1 30>)
    Box Size=<80 80 80>
    MaskSpawnGroup=false
    Module.Override Seed=false
    Rotate Box=false
    Surface Only=false
   )
   &Color(Color=&UniformRangedLinearColor(Minimum=<rgba 1 0.1162 0 1>, Maximum=<rgba 0 0.7329 1 1>))
   &DynamicMaterialParameters(
    Index 0 Param 2=1
    Index 0 Param 4=1
    Module.Param0WriteEnabled=true
    Module.Param1WriteEnabled=true
    Module.Param2WriteEnabled=true
    Module.Param3WriteEnabled=true
   )
  ]
  particleUpdateScript [
   &ParticleState()
   &GravityForce() {enabled=false}
   &SolveForcesAndVelocity(Clamp Velocity=false, Limit Acceleration=false)
   /SetVariables(Particles.Velocity=&ScaleVectorByCurve(
     InitialValue:Vector3f=&UniformRangedVector(Minimum=<0 0 0.5>, Maximum=<0 0 2>)
     ScaleCurve:NiagaraDataInterfaceVectorCurve=DataInterface() {propertyValues {
       XCurve {keys [<at 0 50 RCIM_Cubic>, {arriveTangent=-241.6722, interpMode=RCIM_Cubic, leaveTangent=-241.672, tangentMode=RCTM_User, time=0.5, value=20}]}
       YCurve {keys [<at 0 50 RCIM_Cubic>, {arriveTangent=-247.443, interpMode=RCIM_Cubic, leaveTangent=-247.448, tangentMode=RCTM_User, time=0.4991, value=20}]}
       ZCurve {keys [<at 0 500>,<at 1 1350>]}
      }}
     CurveIndex:NiagaraFloat=$Particles.NormalizedAge
    )) {isSetParameters=true, module=SetVariables_#3}
   &Color(
    Color=&UniformRangedLinearColor(Minimum=<rgba 0 0.0638 1 1>, Maximum=<rgba 1 0.2425 0 1>)
    Scale Alpha=&FloatFromCurve(FloatCurve=DataInterface() {propertyValues {Curve {keys [
         {arriveTangent=0.4553, interpMode=RCIM_Cubic, leaveTangent=0.4553, tangentMode=RCTM_User, time=0.0082, value=0}
         <at 0.2 0.1 RCIM_Cubic>
         <at 1 0 RCIM_Cubic>
        ]}}})
   ) {module=Color001}
  ]
  rendererClasses [&NiagaraSpriteRendererProperties]
  renderers [{rendererClass=&NiagaraSpriteRendererProperties}]
 }
 Little_Round_Drops_Up004 {
  emitterSpawnScript [&FN_PreserveSkydiveUserVariables() {enabled=false, module=PreserveUserVariables}]
  emitterUpdateScript [
   &EmitterState(Inactive Response=::NewEnumerator1, Life Cycle Mode=::NewEnumerator1, Loop Behavior=::NewEnumerator1, Loop Duration=3600, UseLoopDelay=false)
   &SpawnRate(
    SpawnRate=&Multiply_Float(
     A=&UniformRangedFloat(Minimum=1, Maximum=2)
     B=&FloatFromCurve(
      FloatCurve=DataInterface() {propertyValues {Curve {keys [<at 0 0>,<at 1 5>]}}}
      CurveIndex=&Divide_Float(A=&VectorLength(Vector=$Engine.Owner.Velocity), B=3500)
     )
    )
    Use Spawn Probability=false
   )
  ]
  particleSpawnScript [
   &FN_SetSpawnGroup(Emitter Bone Count=8) {enabled=false}
   &InitializeParticle(
    Lifetime=&UniformRangedFloat(Minimum=0.6, Maximum=1)
    Module.Write Color=true
    Module.Write Lifetime=true
    Module.Write Mass=true
    Module.Write Material Random=true
    Module.Write Position=false
    Module.Write Scale=false
    Module.Write SpriteRotation=false
    Module.Write SpriteSize=false
    Position=&SimulationPosition()
   )
   &SampleSkeletalMeshSkeleton(
    Mesh=DataInterface() {propertyValues {
      DefaultMesh=null
      ExcludeBoneName=root
      OptionalComponentClass=null
      PreviewMesh=&M_Med_Soldier_04_Preview
      SkinningMode=SkinOnTheFly
      SoftSourceActor=null
      SourceMode=Default
      UvSetIndex=0
      WholeMeshLOD=-1
      bExcludeBone=false
      bReadDeformedGeometry=true
      bRequireCurrentFrameData=true
      ComponentTags []
      FilteredBones [spine_02]
      FilteredSockets []
      MeshUserParameter {parameter {name=None, typeDefHandle {registeredTypeIndex=1}, varData []}}
      SamplingRegions []
     }}
    Specific Bone Index:NiagaraInt32=$Particles.SpawnGroup
   )
   &SkeletalMeshSkeletonLocation(Module.Inherit Velocity=false, Module.Use Sampled Bone as Ribbon ID=true)
   /SetVariables(Particles.Velocity=&UniformRangedVector(Minimum=<-25 -25 25>, Maximum=<25 25 50>)) {enabled=false, isSetParameters=true, module=SetVariables_#1}
   /SetVariables(Particles.SpriteSize=&MakeVector2D(
     X=&Multiply_Float(
      A=1000
      B=&Lerp_Float(A=1, B=0.4, Alpha=&UniformRangedFloat(Minimum=0.5, Maximum=1))
     )
     Y=&Multiply_Float(
      A=&UniformRangedFloat(Minimum=2000, Maximum=4000)
      B=&FloatFromCurve(
       FloatCurve=DataInterface() {propertyValues {Curve {keys [<at 0 1>,<at 1 2>]}}}
       CurveIndex=&Divide_Float(A=&VectorLength(Vector=$Engine.Owner.Velocity), B=3000)
      )
     )
    )) {isSetParameters=true, module=SetVariables_#2}
   &SphereLocation(
    Coordinate Space:ENiagaraCoordinateSpace=::Simulation
    MaskSpawnGroup=false
    Module.Override Randomness=false
    Sphere Radius=7
    Surface Only=false
   )
   &BoxLocation(
    Box Offset=&UniformRangedVector(Minimum=<0 0 -45>, Maximum=<1 1 0>)
    Box Size=<200 200 80>
    MaskSpawnGroup=false
    Module.Override Seed=false
    Rotate Box=false
    Surface Only=false
   )
   &Color(Color=&LerpLinearColors(StartColor=<rgba 1 1 1 1>, EndColor=<rgba 1 1 1 1>, LerpFactor=0))
   /SetVariables(Particles.Velocity=&MakeVector(
     X=0
     Y=0
     Z=&Multiply_Float(
      A=&UniformRangedFloat(Minimum=450, Maximum=1000)
      B=&FloatFromCurve(
       FloatCurve=DataInterface() {propertyValues {Curve {keys [<at 0 0>,<at 1 2>]}}}
       CurveIndex=&Divide_Float(A=&VectorLength(Vector=$Engine.Owner.Velocity), B=300)
      )
     )
    )) {isSetParameters=true, module=SetVariables_#3}
   /SetVariables(Particles.Velocity=&MakeVector(X=0, Y=0, Z=900)) {enabled=false, isSetParameters=true, module=SetVariables_#4}
  ]
  particleUpdateScript [
   &ParticleState()
   &GravityForce() {enabled=false}
   &SolveForcesAndVelocity(Clamp Velocity=false, Limit Acceleration=false)
   &Color(
    Color=&UniformRangedLinearColor(Minimum=<rgba 1 0.4869 0 1>, Maximum=<rgba 0.0187 0.391 1 1>)
    Scale Alpha=&FloatFromCurve(FloatCurve=DataInterface() {propertyValues {Curve {keys [
         {arriveTangent=0.2541, interpMode=RCIM_Cubic, leaveTangent=0.2541, tangentMode=RCTM_User, time=0.0082, value=0}
         {arriveTangent=-0.0152, interpMode=RCIM_Cubic, leaveTangent=-0.0152, tangentMode=RCTM_User, time=0.2, value=0.028}
         {arriveTangent=-0.0555, interpMode=RCIM_Cubic, leaveTangent=-0.0555, tangentMode=RCTM_User, time=0.4516, value=0.0128}
         <at 1 0 RCIM_Cubic>
        ]}}})
   ) {module=Color001}
  ]
  rendererClasses [&NiagaraSpriteRendererProperties, &NiagaraLightRendererProperties]
  renderers [
   {rendererClass=&NiagaraSpriteRendererProperties}
   {rendererClass=&NiagaraLightRendererProperties, rendererIndex=1}
  ]
 }
}
summary {
 systemName=NS_Contrail_Informer_TowerTapas
 emitters [
  {emitterName=FN_Contrails_RibbonWde, rendererClasses [&NiagaraRibbonRendererProperties]}
  {bEnabled=false, emitterName=Little_Round_Drops_Up, rendererClasses [&NiagaraSpriteRendererProperties]}
  {bEnabled=false, emitterName=Little_Round_Drops_Up001, rendererClasses [&NiagaraSpriteRendererProperties]}
  {bEnabled=false, emitterName=Little_Round_Drops_Up002, rendererClasses [&NiagaraSpriteRendererProperties]}
  {emitterName=FN_AttachedParticle_v3, rendererClasses [&NiagaraSpriteRendererProperties]}
  {
   emitterName=Little_Round_Drops_Up004
   rendererClasses [&NiagaraSpriteRendererProperties, &NiagaraLightRendererProperties]
  }
  {emitterName=FN_NE_Skydive_Mesh, rendererClasses [&NiagaraMeshRendererProperties]}
  {emitterName=FN_Contrails_Light, rendererClasses [&NiagaraLightRendererProperties]}
  {bEnabled=false, emitterName=FN_Contrails_RibbonWde001, rendererClasses [&NiagaraRibbonRendererProperties]}
 ]
 userVariables [
  User.bIsFrontEnd {type=NiagaraBool}
  User.bIsFrontEndPreview {type=NiagaraBool}
  User.PlayerSpeed {type=NiagaraFloat, defaultValue=0, description=""}
  User.RibbonIDOffset {type=NiagaraInt32}
  User.SkeletalMesh {
   type=NiagaraDataInterfaceSkeletalMesh
   defaultValue {
    struct=NiagaraExt_VariableValue_DataInterface
    value {dataInterface=&self:NiagaraDataInterfaceSkeletalMesh_0}
   }
  }
 ]
}
